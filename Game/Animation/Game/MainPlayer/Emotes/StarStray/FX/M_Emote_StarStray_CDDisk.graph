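# Translucent unlit Niagara particle material that draws a procedural CD disc for the StarStray emote.
# Sphere masks on UV distance from center build the disc band, inner and outer edge rings and soft glow, each with an intensity or opacity parameter.
# Radial coordinates drive sheen streaks broken up by panning noise, plus a sine brightness variation.
# Particle color tints the result through MF_FX_EmissiveAdjustments, and particle alpha scales the mask.
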Material /Game/Animation/Game/MainPlayer/Emotes/StarStray/FX/M_Emote_StarStray_CDDisk
graph {
 properties {
  bUsedWithNiagaraMeshParticles=true
  bUsedWithNiagaraSprites=true
  blendMode=BLEND_Translucent
  pixelDepthOffsetMode=PDOM_Legacy
  shadingModel=MSM_Unlit
  twoSided=true
  usedShadingModels=Unlit
 }
 nodes [
  n1 = Constant() {r=0.45}
  n2 = Constant() {r=0.1}
  n3 = Subtract(A=@n1, B=@n2)
  Divide(A=@n3) {constB=1.5}
  n4 = MaterialFunctionCall(Vector or UVs=Add(A=TextureCoordinate()) {constB=-0.5}) {materialFunction=&VectorToRadialValue}
  n5 = ComponentMask(None=@n4.Radial Coordinates) {r=true}
  n6 = ConstantBiasScale(None=Sine(None=@n5) {period=0.25})
  n7 = ParticleColor()
  n8 = SphereMask(A=@n5, B=Constant() {r=0.1}) {attenuationRadius=0.1, hardnessPercent=0}
  n9 = SphereMask(A=@n5, B=Constant() {r=0.5}) {attenuationRadius=0.15, hardnessPercent=0}
  n10 = Multiply(A=@n4.Radial Coordinates, B=Constant2Vector() {g=10, r=1})
  n11 = Subtract(
   A=Multiply(A=Max(A=@n8, B=@n9)) {constB=2}
   B=TextureSample(UVs=Panner(Coordinate=@n10) {speedX=0.5}) {samplerType=SAMPLERTYPE_Grayscale, texture=/FortUILibrary/UIKit/Buttons/Textures/LowResBlurredNoise}.R
  )
  n12 = ScalarParameter() {defaultValue=0.5, parameterName=Sheen Intensity, sliderMax=1}
  n13 = Distance(A=TextureCoordinate(), B=Constant() {r=0.5})
  n14 = SphereMask(A=@n13, B=Divide(A=Add(A=@n1, B=@n2)), Radius=Divide(A=@n3)) {attenuationRadius=0.01}
  n15 = Max(A=SphereMask(A=@n13, B=@n1) {attenuationRadius=0.005}, B=SphereMask(A=@n13, B=@n2) {attenuationRadius=0.01})
  n16 = ScalarParameter() {defaultValue=1, parameterName=Edge_Opacity}
  n17 = Multiply(A=@n14, B=ScalarParameter() {defaultValue=0.1, parameterName=Center Opacity, sliderMax=1}) {constB=0.1}
  n18 = Multiply(
   A=SphereMask(A=@n13, B=@n1) {attenuationRadius=0.05, hardnessPercent=0}
   B=ScalarParameter() {defaultValue=0.1, parameterName=Glow Intensity}
  )
  n19 = Max(A=Min(A=Multiply(A=@n11, B=@n12) {constB=0.5}, B=@n14), B=Max(A=Multiply(A=@n15, B=@n16) {constB=0}, B=Max(A=@n17, B=@n18)))
  MaterialFunctionCall(Emissive=Multiply(A=Add(A=@n6), B=Multiply(A=@n7.RGB, B=@n19))) {materialFunction=&MF_FX_EmissiveAdjustments}
  Multiply(A=@n7.A, B=@n19)
 ]
}
