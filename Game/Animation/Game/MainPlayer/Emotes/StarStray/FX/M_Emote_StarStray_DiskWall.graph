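# Unlit, two-sided translucent Niagara mesh-particle material for the disk wall effect in the StarStray player emote.
# A dynamic Width parameter clips the shape along U, and SphereMasks on U and V draw a Thickness-based border around it.
# A faint U-gradient fills the center, scaled by Center Opacity.
# Particle color and alpha multiply this mask, with the color passing through MF_FX_EmissiveAdjustments.
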Material /Game/Animation/Game/MainPlayer/Emotes/StarStray/FX/M_Emote_StarStray_DiskWall
graph {
 properties {
  bUsedWithNiagaraMeshParticles=true
  blendMode=BLEND_Translucent
  pixelDepthOffsetMode=PDOM_Legacy
  shadingModel=MSM_Unlit
  twoSided=true
  usedShadingModels=Unlit
 }
 nodes [
  n1 = ParticleColor()
  n2 = ComponentMask(None=TextureCoordinate()) {r=true}
  n3 = DynamicParameter() {defaultValue=<rgba 0.8 1 1 1>, paramNames [Width,Param2,Param3,Param4]}
  n4 = ComponentMask(None=TextureCoordinate()) {r=true}
  n5 = ScalarParameter() {defaultValue=0.05, parameterName=Thickness}
  n6 = Divide(A=@n5)
  n7 = Divide(A=@n5)
  n8 = ComponentMask(None=TextureCoordinate()) {r=true}
  n9 = Subtract(A=@n3.Width, B=@n6)
  n10 = ComponentMask(None=TextureCoordinate()) {g=true}
  n11 = Constant() {r=0.5}
  n12 = Divide(A=OneMinus(None=Multiply(A=@n5) {constB=4}))
  n13 = ScalarParameter() {defaultValue=2, parameterName=Top Bottom Width Adjustment, sliderMax=10}
  n14 = Divide(A=@n12, B=@n13)
  n15 = Max(
   A=Max(A=SphereMask(A=@n4, B=@n6, Radius=@n7) {attenuationRadius=0.4}, B=SphereMask(A=@n8, B=@n9, Radius=@n7) {attenuationRadius=0.4})
   B=OneMinus(None=SphereMask(A=@n10, B=@n11, Radius=@n14) {attenuationRadius=0.4})
  )
  n16 = Multiply(A=ScalarParameter() {defaultValue=0.1, parameterName=Center Opacity, sliderMax=1}, B=ComponentMask(None=TextureCoordinate()) {r=true})
  n17 = Multiply(A=If(A=@n2, B=@n3.Width, "A > B"=Constant(), "A < B"=Constant() {r=1}), B=Saturate(None=Add(A=@n15, B=@n16) {constB=0.1}))
  MaterialFunctionCall(Emissive=Multiply(A=@n1.RGB, B=@n17)) {materialFunction=&MF_FX_EmissiveAdjustments}
  Multiply(A=@n12, B=@n13)
  Multiply(A=@n1.A, B=@n17)
  ScalarParameter() {defaultValue=0.75, parameterName=Width, sliderMax=1}
  SphereMask(A=@n2, B=Constant() {r=0.5}, Radius=Divide(A=@n3.Width))
 ]
}
