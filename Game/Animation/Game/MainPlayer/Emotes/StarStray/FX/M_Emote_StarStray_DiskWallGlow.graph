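# Unlit translucent two-sided glow material for the StarStray emote's disk wall FX on Niagara mesh particles, previewed on a cylinder wall mesh.
# Particle color RGB feeds an emissive adjustment function, while particle alpha scales a procedural, texture-free mask.
# The mask multiplies boosted, clamped half-sine falloffs along U and V, fading the edges softly.
# A Width dynamic parameter divides U, letting particles control the glow's horizontal extent.
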Material /Game/Animation/Game/MainPlayer/Emotes/StarStray/FX/M_Emote_StarStray_DiskWallGlow
graph {
 properties {
  bUsedWithNiagaraMeshParticles=true
  blendMode=BLEND_Translucent
  pixelDepthOffsetMode=PDOM_Legacy
  previewMesh=&/SM_Emote_StarStray_FXCylinderWall
  shadingModel=MSM_Unlit
  twoSided=true
  usedShadingModels=Unlit
 }
 nodes [
  n1 = ParticleColor()
  MaterialFunctionCall(Emissive=@n1.RGB) {materialFunction=&MF_FX_EmissiveAdjustments}
  n2 = TextureCoordinate()
  n3 = Divide(A=ComponentMask(None=@n2) {r=true}, B=DynamicParameter() {defaultValue=<rgba 0.8 1 1 1>, paramNames [Width,Param2,Param3,Param4]}.Width)
  n4 = Constant() {r=2}
  n5 = Multiply(A=Sine(None=Saturate(None=@n3)) {period=2}, B=@n4) {constB=5}
  n6 = Multiply(A=Sine(None=ComponentMask(None=@n2) {g=true}) {period=2}, B=@n4) {constB=5}
  Multiply(A=@n1.A, B=Multiply(A=Saturate(None=@n5), B=Saturate(None=@n6)))
 ]
}
