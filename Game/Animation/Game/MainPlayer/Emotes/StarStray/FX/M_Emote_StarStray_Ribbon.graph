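# Unlit translucent Niagara ribbon material for the main player's StarStray emote FX, drawing a soft glowing streak.
# Particle color RGB passes through the shared MF_FX_EmissiveAdjustments function for emissive.
# Opacity is procedural and texture-free: two sphere masks centered on the UV's V coordinate add a narrow bright core to a faint, wide falloff.
# Particle alpha scales the combined mask, so the particle system controls tint and fade.
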Material /Game/Animation/Game/MainPlayer/Emotes/StarStray/FX/M_Emote_StarStray_Ribbon
graph {
 properties {bUsedWithNiagaraRibbons=true, blendMode=BLEND_Translucent, pixelDepthOffsetMode=PDOM_Legacy, shadingModel=MSM_Unlit, usedShadingModels=Unlit}
 nodes [
  n1 = ParticleColor()
  MaterialFunctionCall(Emissive=@n1.RGB) {materialFunction=&MF_FX_EmissiveAdjustments}
  n2 = ComponentMask(None=TextureCoordinate()) {g=true}
  n3 = Constant() {r=0.5}
  n4 = SphereMask(A=@n2, B=@n3, Radius=@n3) {hardnessPercent=0}
  n5 = Constant() {r=0.5}
  n6 = Constant() {r=0.1}
  Multiply(A=@n1.A, B=Add(A=Multiply(A=@n4) {constB=0.25}, B=SphereMask(A=@n2, B=@n5, Radius=@n6) {hardnessPercent=50}))
 ]
}
