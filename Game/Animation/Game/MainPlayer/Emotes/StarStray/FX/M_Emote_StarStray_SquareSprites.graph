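# Unlit, translucent, two-sided Niagara sprite material for the StarStray player emote effects, drawing procedural squares without textures.
# Particle color RGB feeds MF_FX_EmissiveAdjustments for the glow.
# Separate 1D SphereMasks on U and V, inverted and combined with Max, form a square with bright borders.
# An added bias keeps the center partly filled before the shape is saturated and multiplied by particle alpha.
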Material /Game/Animation/Game/MainPlayer/Emotes/StarStray/FX/M_Emote_StarStray_SquareSprites
graph {
 properties {
  bUsedWithNiagaraSprites=true
  blendMode=BLEND_Translucent
  pixelDepthOffsetMode=PDOM_Legacy
  shadingModel=MSM_Unlit
  twoSided=true
  usedShadingModels=Unlit
 }
 nodes [
  n1 = ParticleColor()
  MaterialFunctionCall(Emissive=@n1.RGB) {materialFunction=&MF_FX_EmissiveAdjustments}
  n2 = TextureCoordinate()
  n3 = ComponentMask(None=@n2) {r=true}
  n4 = Constant() {r=0.5}
  n5 = ComponentMask(None=@n2) {g=true}
  n6 = Constant() {r=0.5}
  n7 = Max(
   A=OneMinus(None=SphereMask(A=@n3, B=@n4) {attenuationRadius=0.45, hardnessPercent=90})
   B=OneMinus(None=SphereMask(A=@n5, B=@n6) {attenuationRadius=0.45, hardnessPercent=90})
  )
  Multiply(A=@n1.A, B=Saturate(None=Add(A=@n7) {constB=0.25}))
 ]
}
