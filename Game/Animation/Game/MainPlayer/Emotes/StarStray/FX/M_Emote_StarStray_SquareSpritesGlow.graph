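# Translucent, unlit, two-sided glow material for Niagara sprites in the StarStray player emote, named for square sprites.
# Particle color RGB feeds a shared FX emissive adjustment function.
# A soft square falloff comes from half-period sine waves on U and V, each halved and multiplied together, then scaled by particle alpha, so no texture is sampled.
# Two Saturate nodes on the sine results feed nothing.
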Material /Game/Animation/Game/MainPlayer/Emotes/StarStray/FX/M_Emote_StarStray_SquareSpritesGlow
graph {
 properties {
  bUsedWithNiagaraSprites=true
  blendMode=BLEND_Translucent
  pixelDepthOffsetMode=PDOM_Legacy
  shadingModel=MSM_Unlit
  twoSided=true
  usedShadingModels=Unlit
 }
 nodes [
  n1 = ParticleColor()
  MaterialFunctionCall(Emissive=@n1.RGB) {materialFunction=&MF_FX_EmissiveAdjustments}
  n2 = TextureCoordinate()
  n3 = Constant() {r=0.5}
  n4 = Multiply(A=Sine(None=ComponentMask(None=@n2) {r=true}) {period=2}, B=@n3) {constB=5}
  n5 = Multiply(A=Sine(None=ComponentMask(None=@n2) {g=true}) {period=2}, B=@n3) {constB=5}
  Multiply(A=@n1.A, B=Multiply(A=@n4, B=@n5))
  Saturate(None=@n4)
  Saturate(None=@n5)
 ]
}
