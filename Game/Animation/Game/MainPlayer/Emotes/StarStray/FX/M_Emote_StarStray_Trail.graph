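# Additive unlit trail material for the main player's StarStray emote, usable on Niagara ribbons and sprites.
# Particle color RGB feeds MF_FX_EmissiveAdjustments, while particle alpha scales a texture-free mask built from UVs.
# The mask fades along U with a short lead-in and combines a soft-edged band across V with two brighter lines near its edges, sized by the Thickness parameter.
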Material /Game/Animation/Game/MainPlayer/Emotes/StarStray/FX/M_Emote_StarStray_Trail
graph {
 properties {
  bEnableResponsiveAA=true
  bOutputTranslucentVelocity=true
  bUsedWithNiagaraRibbons=true
  bUsedWithNiagaraSprites=true
  blendMode=BLEND_Additive
  pixelDepthOffsetMode=PDOM_Legacy
  shadingModel=MSM_Unlit
  usedShadingModels=Unlit
 }
 nodes [
  n1 = ParticleColor()
  MaterialFunctionCall(Emissive=@n1.RGB) {materialFunction=&MF_FX_EmissiveAdjustments}
  n2 = ComponentMask(None=TextureCoordinate()) {r=true}
  n3 = Subtract(A=ComponentMask(None=TextureCoordinate()) {g=true}) {constB=0.5}
  n4 = ComponentMask(None=TextureCoordinate()) {g=true}
  n5 = Constant() {r=0.5}
  n6 = Multiply(
   A=OneMinus(None=@n2)
   B=If(
    A=Abs(None=@n3)
    B=Constant() {r=0.4}
    "A > B"=SphereMask(A=@n4, B=@n5) {attenuationRadius=0.45, hardnessPercent=50}
    "A < B"=Constant() {r=0.8}
   )
  )
  n7 = ScalarParameter() {defaultValue=0.03, parameterName=Thickness, sliderMax=1}
  n8 = SphereMask(A=ComponentMask(None=TextureCoordinate()) {g=true}, B=Constant() {r=0.1}, Radius=@n7) {attenuationRadius=0.02, hardnessPercent=0}
  n9 = SphereMask(A=ComponentMask(None=TextureCoordinate()) {g=true}, B=Constant() {r=0.9}, Radius=@n7) {attenuationRadius=0.2, hardnessPercent=0}
  n10 = Add(A=Multiply(A=Add(A=@n8, B=@n9)) {constB=2}) {constB=0.25}
  Multiply(A=@n1.A, B=Multiply(A=SmoothStep(Value=@n2) {constMax=0.05}, B=Multiply(A=@n6, B=@n10)))
 ]
}
