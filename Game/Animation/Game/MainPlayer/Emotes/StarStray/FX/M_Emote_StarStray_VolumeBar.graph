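# Translucent unlit material for Niagara mesh particles that draws an animated, rainbow equalizer-style volume bar display for the StarStray emote.
# UVs are quantized into rows and columns, and each row is thresholded against a per-column wave of summed, scrolling sines with a periodic beat.
# A rainbow curve atlas colors rows, and its alpha adds a per-column threshold.
# Masks cut cell gaps; depth fade is optional, skipped on mobile.
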Material /Game/Animation/Game/MainPlayer/Emotes/StarStray/FX/M_Emote_StarStray_VolumeBar
graph {
 properties {
  bUsedWithNiagaraMeshParticles=true
  blendMode=BLEND_Translucent
  pixelDepthOffsetMode=PDOM_Legacy
  shadingModel=MSM_Unlit
  twoSided=true
  usedShadingModels=Unlit
 }
 nodes [
  n1 = Ceil(None=ScalarParameter() {defaultValue=10, parameterName=Rows, sliderMax=100})
  n2 = Multiply(A=ComponentMask(None=TextureCoordinate()) {g=true}, B=@n1)
  n3 = Divide(A=Floor(None=@n2), B=@n1)
  n4 = Multiply(A=NamedRerouteUsage()) {constB=1.25}
  n5 = Constant() {r=1}
  n6 = Constant()
  n7 = Add(A=Multiply(A=If(A=@n3, B=@n4, "A > B"=@n5, "A < B"=@n6), B=ScalarParameter() {defaultValue=0.1, parameterName=Emissive Boost, sliderMax=5}))
  n8 = TextureSample(UVs=@n3) {samplerType=SAMPLERTYPE_LinearColor, texture=&/CurveAtlas_Emote_StarStray_Rainbow}
  n9 = ParticleColor()
  MaterialFunctionCall(Emissive=Multiply(A=Multiply(A=@n7, B=@n8.RGB), B=@n9.RGB)) {materialFunction=&MF_FX_EmissiveAdjustments}
  n10 = Time()
  n11 = Sine(None=@n10) {period=0.4839}
  n12 = Multiply(
   A=If(A=@n11, "A > B"=@n11, "A == B"=@n11, "A < B"=Constant())
   B=ScalarParameter() {defaultValue=0.2, parameterName=Beat Intensity, sliderMax=1}
  ) {constB=0.25}
  n13 = Multiply(A=@n10, B=ScalarParameter() {defaultValue=0.2, parameterName=Speed, sliderMax=5})
  n14 = Ceil(None=ScalarParameter() {defaultValue=10, parameterName=Columns, sliderMax=20})
  n15 = Multiply(A=ComponentMask(None=TextureCoordinate()) {r=true}, B=@n14)
  n16 = Divide(A=Floor(None=@n15), B=@n14)
  n17 = Frac(None=Add(A=@n13, B=@n16))
  n18 = Multiply(A=ConstantBiasScale(None=Sine(None=@n17) {period=0.3})) {constB=0.25}
  n19 = Subtract(A=@n13, B=@n16)
  n20 = ConstantBiasScale(None=Sine(None=Frac(None=@n19)) {period=0.5})
  n21 = Sine(None=@n17) {period=0.8}
  n22 = Divide(A=Add(A=Add(A=@n18, B=@n20), B=ConstantBiasScale(None=@n21))) {constB=4}
  NamedRerouteDeclaration(None=Multiply(A=OneMinus(None=@n12), B=Add(A=@n22) {constB=0.3})) {name=Wave, nodeColor=<rgba 1 0.9609 0 1>}
  n23 = TextureSample(UVs=Frac(None=@n16)) {samplerType=SAMPLERTYPE_LinearColor, texture=&CurveAtlas_Emote_StarStray_Rainbow}
  n24 = Constant() {r=1}
  n25 = Constant()
  n26 = If(A=@n3, B=NamedRerouteUsage(), "A > B"=Constant() {r=1}, "A < B"=Constant())
  n27 = Frac(None=@n2)
  n28 = Divide(A=ScalarParameter() {defaultValue=0.1, parameterName=Row Gap Width, sliderMax=1})
  n29 = Frac(None=@n15)
  n30 = Divide(A=ScalarParameter() {defaultValue=0.1, parameterName=Column Gap Width, sliderMax=1})
  n31 = Min(
   A=OneMinus(None=SphereMask(A=@n27, B=@n28, Radius=@n28) {attenuationRadius=0.05})
   B=OneMinus(None=SphereMask(A=@n29, B=@n30, Radius=@n30) {attenuationRadius=0.05})
  )
  n32 = Multiply(A=@n9.A, B=Multiply(A=If(A=@n3, B=@n23.A, "A > B"=@n24, "A < B"=@n25), B=Min(A=@n26, B=@n31)))
  n33 = ScalarParameter() {defaultValue=10, parameterName=DepthFadeDistance, sliderMax=100}
  ShadingPathSwitch(
   Default=StaticSwitchParameter(True=DepthFade(Opacity=@n32, FadeDistance=@n33), False=@n32) {defaultValue=true, parameterName=UseDepthFade}
   Mobile=@n32
  ) {desc=Migrated from FeatureLevelSwitch}
 ]
}
